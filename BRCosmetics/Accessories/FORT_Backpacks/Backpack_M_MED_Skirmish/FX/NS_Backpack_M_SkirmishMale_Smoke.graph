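# Niagara smoke effect for the male Skirmish backpack cosmetic, built from three burst sprite emitters.
# Two white layers push puffs along +X with sideways spread, then drag, gravity, growth curves and alpha fades shape them as SubUV animation plays.
# A slightly delayed, faint grey layer falls downward, stretching vertically through non-uniform sprite scaling as it fades in and out.
# All layers use camera offsets, and the system exposes no user parameters.
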NiagaraSystem /BRCosmetics/Accessories/FORT_Backpacks/Backpack_M_MED_Skirmish/FX/NS_Backpack_M_SkirmishMale_Smoke
dependencies {usedDataInterfaces [&NiagaraDataInterfaceCurve], usedDynamicInputs [], usedModules [&SystemState]}
emitters {
 Empty {
  emitterUpdateScript [
   &EmitterState(Life Cycle Mode=::NewEnumerator1, Loop Behavior=::NewEnumerator1, Loop Duration=1, UseLoopDelay=false)
   &SpawnBurst_Instantaneous(Spawn Count=10, Use Spawn Probability=false)
  ]
  particleSpawnScript [
   &InitializeParticle(
    Color Mode=::NewEnumerator0
    Lifetime=&RandomRangeFloat(Minimum=0.4, Maximum=0.65)
    Sprite Rotation Angle=&RandomRangeFloat(Minimum=0, Maximum=1)
    Sprite Rotation Mode=::NewEnumerator1
    Sprite Size Mode=::NewEnumerator3
    Uniform Sprite Size=&RandomRangeFloat(Minimum=28, Maximum=40)
    UsePositionOffset=false
    Write Material Random=true
   )
   &AddVelocity(Velocity=&RandomRangeVector(Minimum=<20 -40 0>, Maximum=<35 40 0>))
   &Color(Color=<rgba 1 1 1 1>)
   &CylinderLocation(
    Cylinder Height=14
    Cylinder Radius=1
    MaskSpawnGroup=false
    Module.Enable Lathe Profile=false
    Module.Override Randomness=false
    Module.Surface Only=false
    Offset=<0 0 -4>
   )
  ]
  particleUpdateScript [
   &GravityForce(Gravity=<0 0 -86>)
   &Drag(Drag=3, Use Linear Drag=true, Use Rotational Drag=false)
   &SolveForcesAndVelocity(Clamp Velocity=false, Limit Acceleration=false)
   &ParticleState()
   &ScaleColor(
    Scale Alpha=&FloatFromCurve(FloatCurve=DataInterface() {propertyValues {Curve {keys [
         {arriveTangent=-0.5828, interpMode=RCIM_Cubic, leaveTangent=-0.5828, tangentMode=RCTM_User, time=0, value=1}
         {arriveTangent=-0.4276, interpMode=RCIM_Cubic, leaveTangent=-0.4276, tangentMode=RCTM_User, time=1, value=0}
        ]}}})
    ScaleA=true
    ScaleRGB=true
   )
   &ScaleSpriteSize(Scale Factor=&Vector2DFromFloat(Value=&FloatFromCurve(FloatCurve=DataInterface() {propertyValues {Curve {keys [
          {arriveTangent=5.8882, interpMode=RCIM_Cubic, leaveTangent=2.6618, tangentMode=RCTM_User, time=0.0024, value=0}
          {arriveTangent=1.3031, interpMode=RCIM_Cubic, leaveTangent=1.3031, time=0.3937, value=0.9205}
          {interpMode=RCIM_Cubic, leaveTangent=-0.0928, time=1, value=1.3}
         ]}}})))
   &CameraOffset(Camera Offset Amount=20)
   &SubUVAnimation()
  ]
  rendererClasses [&NiagaraSpriteRendererProperties]
  renderers [{rendererClass=&NiagaraSpriteRendererProperties}]
 }
 Empty001 {
  emitterUpdateScript [
   &EmitterState(Life Cycle Mode=::NewEnumerator1, Loop Behavior=::NewEnumerator1, Loop Duration=1, UseLoopDelay=false)
   &SpawnBurst_Instantaneous(Spawn Count=4, Spawn Time=0.15, Use Spawn Probability=false)
  ]
  particleSpawnScript [
   &InitializeParticle(
    Color Mode=::NewEnumerator0
    Lifetime=&RandomRangeFloat(Minimum=0.3, Maximum=0.6)
    Sprite Rotation Angle=&RandomRangeFloat(Minimum=0, Maximum=1)
    Sprite Rotation Mode=::NewEnumerator1
    Sprite Size Mode=::NewEnumerator3
    Uniform Sprite Size=&RandomRangeFloat(Minimum=10, Maximum=20)
    UsePositionOffset=false
    Write Material Random=true
   )
   &AddVelocity(Velocity=&RandomRangeVector(Minimum=<0 0 -20>, Maximum=<0 0 -55>))
   &Color(Color=<rgba 0.2708 0.2708 0.2708 0.3>)
   &SphereLocation(
    MaskSpawnGroup=false
    Offset=<5 0 -5>
    Override Seed=false
    Sphere Radius=5
    Surface Only=false
    UseNonUniformScale=false
    UseOffset=true
    UseSphereOrientation=false
   )
  ]
  particleUpdateScript [
   &GravityForce(Gravity=<0 0 -150>)
   &Drag(Drag=1.4, Use Linear Drag=true, Use Rotational Drag=false)
   &SolveForcesAndVelocity(Clamp Velocity=false, Limit Acceleration=false)
   &ParticleState()
   &ScaleColor(
    Scale Alpha=&FloatFromCurve(FloatCurve=DataInterface() {propertyValues {Curve {keys [
         {arriveTangent=5.4481, interpMode=RCIM_Cubic, leaveTangent=5.4481, tangentMode=RCTM_User, time=0, value=0}
         <at 0.35 1 RCIM_Cubic>
         {arriveTangent=-2.0396, interpMode=RCIM_Cubic, leaveTangent=-4.7808, tangentMode=RCTM_User, time=1, value=0}
        ]}}})
    ScaleA=true
    ScaleRGB=true
   )
   &ScaleSpriteSize(Scale Factor=&MakeVector2D(
     X=0.4
     Y=&FloatFromCurve(FloatCurve=DataInterface() {propertyValues {Curve {keys [<at 0 2>,<at 0.9635 3.5>]}}})
    ))
   &CameraOffset(Camera Offset Amount=10)
  ]
  rendererClasses [&NiagaraSpriteRendererProperties]
  renderers [{rendererClass=&NiagaraSpriteRendererProperties}]
 }
 Empty002 {
  emitterUpdateScript [
   &EmitterState(Life Cycle Mode=::NewEnumerator1, Loop Behavior=::NewEnumerator1, Loop Duration=1, UseLoopDelay=false)
   &SpawnBurst_Instantaneous(Spawn Count=6, Use Spawn Probability=false)
  ]
  particleSpawnScript [
   &InitializeParticle(
    Color Mode=::NewEnumerator0
    Lifetime=&RandomRangeFloat(Minimum=0.7, Maximum=0.86)
    Sprite Rotation Angle=&RandomRangeFloat(Minimum=0, Maximum=360)
    Sprite Rotation Mode=::NewEnumerator1
    Sprite Size Mode=::NewEnumerator3
    Uniform Sprite Size=&RandomRangeFloat(Minimum=20, Maximum=35)
    UsePositionOffset=false
    Write Material Random=true
   )
   &AddVelocity(Velocity=&RandomRangeVector(Minimum=<10 -20 0>, Maximum=<20 20 0>))
   &Color(Color=<rgba 1 1 1 1.2>)
   &CylinderLocation(
    Cylinder Height=4
    Cylinder Radius=0
    MaskSpawnGroup=false
    Module.Enable Lathe Profile=false
    Module.Override Randomness=false
    Module.Surface Only=false
    Offset=<0 0 -10>
   )
  ]
  particleUpdateScript [
   &GravityForce(Gravity=<0 0 -110>)
   &Drag(Drag=3, Use Linear Drag=true, Use Rotational Drag=false)
   &SolveForcesAndVelocity(Clamp Velocity=false, Limit Acceleration=false)
   &ParticleState()
   &ScaleColor(
    Scale Alpha=&FloatFromCurve(FloatCurve=DataInterface() {propertyValues {Curve {keys [<at 0 1 RCIM_Cubic>, {arriveTangent=-2.0509, interpMode=RCIM_Cubic, leaveTangent=-2.0509, tangentMode=RCTM_User, time=1, value=0}]}}})
    ScaleA=true
    ScaleRGB=true
   )
   &ScaleSpriteSize(Scale Factor=&Vector2DFromFloat(Value=&FloatFromCurve(FloatCurve=DataInterface() {propertyValues {Curve {keys [
          {arriveTangent=0.3913, interpMode=RCIM_Cubic, leaveTangent=0.3913, tangentMode=RCTM_User, time=0.0024, value=0.1665}
          {arriveTangent=1.1924, interpMode=RCIM_Cubic, leaveTangent=1.1924, tangentMode=RCTM_User, time=0.4625, value=0.4625}
          {arriveTangent=2.4392, interpMode=RCIM_Cubic, leaveTangent=2.4392, tangentMode=RCTM_User, time=1, value=1.3}
         ]}}})))
   &CameraOffset(Camera Offset Amount=10)
   &SubUVAnimation()
  ]
  rendererClasses [&NiagaraSpriteRendererProperties]
  renderers [{rendererClass=&NiagaraSpriteRendererProperties}]
 }
}
summary {
 systemName=NS_Backpack_M_SkirmishMale_Smoke
 emitters [
  {emitterName=Empty001, rendererClasses [&NiagaraSpriteRendererProperties]}
  {emitterName=Empty002, rendererClasses [&NiagaraSpriteRendererProperties]}
  {emitterName=Empty, rendererClasses [&NiagaraSpriteRendererProperties]}
 ]
 userVariables []
}
