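# Physics asset for the M_MED_Skirmish backpack cosmetic, defining collision bodies and joints for its bag and string bones.
# From the pack body, a two-link bottom chain uses limited joints, while the mid bag rotates freely and carries the tightly limited top.
# The top anchors four twist-locked end pieces and two three-link string chains.
# Boxes cover most bones, with capsules on the top and bottom sections, the bottom ones zero-length spheres.
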PhysicsAsset /BRCosmetics/Accessories/FORT_Backpacks/Backpack_M_MED_Skirmish/Meshes/M_MED_Skirmish_Pack_Physics
bodies {
 dyn_bag_bottom_01 [{shapeType=Capsule, length=0, radius=10.5955, center=<10.34 -0.0006 -0.0462>, rotation=<rot 0.1727 0 0>}]
 dyn_bag_bottom_02 [{shapeType=Capsule, length=0, radius=10.5955, center=<10.6004 -0.0001 -0.0454>, rotation=<rot 0.1727 0 0>}]
 dyn_bag_ends_02 [{shapeType=Box, extentX=4, extentY=4, extentZ=4, center=<2.3413 0 0>}]
 dyn_bag_ends_03 [{shapeType=Box, extentX=4, extentY=4, extentZ=4, center=<2.3413 0 0>}]
 dyn_bag_ends_04 [{shapeType=Box, extentX=4, extentY=4, extentZ=4, center=<2.3413 0 0>}]
 dyn_bag_ends_05 [{shapeType=Box, extentX=4, extentY=4, extentZ=4, center=<2.3413 0 0>}]
 dyn_bag_mid [{shapeType=Box, extentX=4, extentY=4, extentZ=4}]
 dyn_bag_top [{shapeType=Capsule, length=10.7646, radius=9.9851, center=<-3.3328 0.7431 -0.1189>, rotation=<rot 29.6361 90.2687 -90>}]
 pack [{shapeType=Box, extentX=4, extentY=4, extentZ=4}]
 string_01 [{shapeType=Box, extentX=2.25, extentY=1.2632, extentZ=1.4153, center=<1.325 0.1242 0.0418>, rotation=<rot 0 0.9244 -73.5531>}]
 string_02 [{shapeType=Box, extentX=1.5435, extentY=1.0508, extentZ=2.5348, center=<1.2937 -0.0135 -0.1466>, rotation=<rot 0.0001 93.0993 -88.9594>}]
 string_03 [{shapeType=Box, extentX=1.5517, extentY=1.6255, extentZ=2.8302, center=<1.8931 -0.3122 0.0213>, rotation=<rot 0 -98.7418 -88.79>}]
 string_04 [{shapeType=Box, extentX=0.75, extentY=0.8453, extentZ=4.922, center=<2.6904 0.0645 -0.4914>, rotation=<rot 0 -89.0177 -84.2049>}]
 string_05 [{shapeType=Box, extentX=1.3149, extentY=1.6088, extentZ=3.6773, center=<1.8307 -0.1327 -0.4648>, rotation=<rot 0 -97.113 -75.2446>}]
 string_06 [{shapeType=Box, extentX=1.2287, extentY=1.6487, extentZ=2.9244, center=<1.9093 0.0208 -0.2504>, rotation=<rot 0.2763 92.6056 -92.3802>}]
}
constraints [
 {bone1Name=string_01, bone2Name=dyn_bag_top, swing1LimitDegrees=10, swing2LimitDegrees=10, twistLimitDegrees=45, twistMotion=Locked}
 {bone1Name=string_02, bone2Name=string_01, swing1LimitDegrees=35, swing2LimitDegrees=35, twistLimitDegrees=5}
 {bone1Name=string_03, bone2Name=string_02, swing1LimitDegrees=35, swing2LimitDegrees=35, twistLimitDegrees=5}
 {bone1Name=string_04, bone2Name=dyn_bag_top, swing1LimitDegrees=35, swing2LimitDegrees=35, twistLimitDegrees=5}
 {bone1Name=string_05, bone2Name=string_04, swing1LimitDegrees=35, swing2LimitDegrees=35, twistLimitDegrees=5}
 {bone1Name=string_06, bone2Name=string_05, swing1LimitDegrees=35, swing2LimitDegrees=35, twistLimitDegrees=5}
 {bone1Name=dyn_bag_ends_02, bone2Name=dyn_bag_top, swing1LimitDegrees=45, swing2LimitDegrees=20, twistLimitDegrees=45, twistMotion=Locked}
 {bone1Name=dyn_bag_ends_03, bone2Name=dyn_bag_top, swing1LimitDegrees=45, swing2LimitDegrees=20, twistLimitDegrees=45, twistMotion=Locked}
 {bone1Name=dyn_bag_ends_04, bone2Name=dyn_bag_top, swing1LimitDegrees=45, swing2LimitDegrees=20, twistLimitDegrees=45, twistMotion=Locked}
 {bone1Name=dyn_bag_ends_05, bone2Name=dyn_bag_top, swing1LimitDegrees=45, swing2LimitDegrees=20, twistLimitDegrees=45, twistMotion=Locked}
 {bone1Name=dyn_bag_top, bone2Name=dyn_bag_mid, swing1LimitDegrees=13, swing2LimitDegrees=13, twistLimitDegrees=1}
 {
  bone1Name=dyn_bag_mid
  bone2Name=pack
  swing1LimitDegrees=45
  swing1Motion=Free
  swing2LimitDegrees=45
  swing2Motion=Free
  twistLimitDegrees=45
  twistMotion=Free
 }
 {bone1Name=dyn_bag_bottom_01, bone2Name=pack, swing1LimitDegrees=15, swing2LimitDegrees=25, twistLimitDegrees=4}
 {bone1Name=dyn_bag_bottom_02, bone2Name=dyn_bag_bottom_01, swing1LimitDegrees=20, swing2LimitDegrees=35, twistLimitDegrees=4}
]
